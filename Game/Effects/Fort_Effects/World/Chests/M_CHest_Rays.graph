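# Unlit translucent mesh-particle material for light rays in chest effects.
# Particle color passes through DayNight_Auto_Bloom_Control and MF_FX_EmissiveAdjustments to form the emissive.
# Opacity is particle alpha times the cubed V texture coordinate, fading the rays along their UV length, with a depth fade softening intersections.
# A ramp texture sample, a one-minus and a squared-gradient multiply feed no other node.
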Material /Game/Effects/Fort_Effects/World/Chests/M_CHest_Rays
graph {
 properties {bUsedWithMeshParticles=true, blendMode=BLEND_Translucent, pixelDepthOffsetMode=PDOM_Legacy, shadingModel=MSM_Unlit, usedShadingModels=Unlit}
 nodes [
  n1 = ParticleColor()
  n2 = ComponentMask(None=TextureCoordinate()) {g=true}
  DepthFade(Opacity=Multiply(A=@n1.A, B=Power(Base=@n2) {constExponent=3}))
  MaterialFunctionCall(Emissive=MaterialFunctionCall!(DayBloom_Mult=Constant() {r=1}, In=@n1.RGB) {materialFunction=/Game/Packages/Fortnite_UniqueMaterials_1/MH_Materials/DayNight_Auto_Bloom_Control}.Result) {materialFunction=&MF_FX_EmissiveAdjustments}
  Multiply(A=@n2, B=@n2)
  OneMinus(None=@n2)
  TextureSample() {texture=/Game/Effects/Fort_Effects/Textures/Ramps/T_Ramp_Smooth_Reverse}
 ]
}
