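# Unlit translucent mesh-particle material for the semicircular glow card on chests, with emissive driven by particle color.
# Opacity combines particle alpha, a sphere mask on UV1 and vertical gradients from both UV sets.
# The default shading path adds depth and near-camera fades, while mobile replaces them with a constant dimming factor.
# A static switch, on by default, fades the bottom using world height relative to particle position.
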Material /Game/Effects/Fort_Effects/World/Chests/M_ChestGlowSemiCircleMeshCard
graph {
 properties {
  bUsedWithMeshParticles=true
  bUsedWithNanite=true
  bUsedWithNiagaraMeshParticles=true
  blendMode=BLEND_Translucent
  pixelDepthOffsetMode=PDOM_Legacy
  shadingModel=MSM_Unlit
  usedShadingModels=Unlit
 }
 nodes [
  n1 = ParticleColor()
  n2 = MaterialFunctionCall(Emissive=@n1.RGB) {materialFunction=&MF_FX_EmissiveAdjustments}
  MaterialFunctionCall(Emissive=@n2.Result) {materialFunction=&MF_FX_EmissiveAdjustments}
  Reroute(None=@n2.Result)
  n3 = TextureCoordinate() {coordinateIndex=1}
  n4 = ComponentMask(None=VectorParameter() {defaultValue=<rgba 0.5 0.5 0 1>}.RGB) {g=true, r=true}
  n5 = ScalarParameter() {defaultValue=0.5, parameterName=Radius}
  n6 = ScalarParameter() {parameterName=Hardness}
  n7 = Multiply(A=@n1.A, B=Power(Base=SphereMask(A=@n3, B=@n4, Radius=@n5, Hardness=@n6), Exp=ScalarParameter() {defaultValue=3, parameterName=Exp}))
  n8 = Multiply(A=@n7, B=DepthFade(FadeDistance=ScalarParameter() {defaultValue=250, parameterName=Depth Fade}))
  n9 = MaterialFunctionCall(Distance=ScalarParameter() {defaultValue=500, parameterName=Camera Fade}) {materialFunction=&Near_Camera_Fade}
  n10 = OneMinus(None=ComponentMask(None=TextureCoordinate()) {g=true})
  n11 = ScalarParameter() {defaultValue=3, parameterName=Exp}
  n12 = OneMinus(None=ComponentMask(None=TextureCoordinate() {coordinateIndex=1}) {g=true})
  n13 = ScalarParameter() {defaultValue=3, parameterName=ExpCoord1}
  n14 = Multiply(
   A=ShadingPathSwitch(Default=Multiply(A=@n8, B=@n9.Result), Mobile=Multiply(A=@n7) {constB=0.075}) {desc=Migrated from FeatureLevelSwitch}
   B=Multiply(A=Power(Base=@n10, Exp=@n11), B=Power(Base=@n12, Exp=@n13))
  )
  n15 = Subtract(A=WorldPosition().XYZ, B=ParticlePositionWS())
  n16 = Saturate(None=Multiply(A=ComponentMask(None=@n15) {b=true}, B=ScalarParameter() {defaultValue=0.05, parameterName=FadeBottomWorldScaleMultiplier}) {constB=0.01})
  Reroute(None=StaticSwitchParameter(True=Multiply(A=@n14, B=@n16), False=@n14) {defaultValue=true, parameterName=FadeBottom})
 ]
}
