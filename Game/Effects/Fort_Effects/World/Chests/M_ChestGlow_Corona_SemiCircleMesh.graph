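# Additive unlit material for mesh particles that draws the flickering, spiky corona glow of a chest on a semicircle mesh.
# Two counter-rotating spikeball samples combine with a sphere-mask-driven halo gradient, tinted by particle color, then pass through HDR, desaturation and hue adjustments.
# Dynamic parameters control size, rotation and layered-sine flicker.
# Particle alpha, a vertical UV falloff and non-mobile depth and near-camera fades form a separate fade term.
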Material /Game/Effects/Fort_Effects/World/Chests/M_ChestGlow_Corona_SemiCircleMesh
graph {
 properties {
  bUsedWithMeshParticles=true
  bUsedWithNanite=true
  bUsedWithNiagaraMeshParticles=true
  blendMode=BLEND_Additive
  pixelDepthOffsetMode=PDOM_Legacy
  shadingModel=MSM_Unlit
  usedShadingModels=Unlit
 }
 nodes [
  n1 = Multiply(A=Time())
  n2 = Multiply(A=Sine(None=@n1) {period=0.155}, B=Sine(None=@n1) {period=0.2})
  n3 = Sine(None=@n1) {period=0.43}
  n4 = Clamp(None=Multiply(A=Sine(None=@n1) {period=0.075}, B=Multiply(A=@n2, B=@n3))) {maxDefault=3}
  n5 = DynamicParameter() {defaultValue=<rgba 0.6 1 1 1>, paramNames [Size,RotateSpeed,Flicker,Param4]}
  n6 = ParticleColor()
  n7 = TextureCoordinate() {coordinateIndex=1, uTiling=0.25, vTiling=0.25}
  n8 = Multiply(A=Time(), B=@n5.RotateSpeed)
  n9 = TextureSample(UVs=Rotator(Coordinate=Add(A=@n7) {constB=0.375}, Time=@n8)) {texture=/Game/Effects/Fort_Effects/Textures/Flares/T_Flare_Spikeball}
  n10 = TextureCoordinate() {coordinateIndex=1, uTiling=0.15, vTiling=0.15}
  n11 = TextureSample(UVs=Rotator(Coordinate=Add(A=@n10) {constB=0.425}, Time=@n8) {speed=-0.72}) {texture=&T_Flare_Spikeball}
  n12 = Multiply(A=@n9.RGB, B=Power(Base=OneMinus(None=@n11.RGB)) {constExponent=5})
  n13 = Multiply(A=ScalarParameter() {defaultValue=1, parameterName=Brightness}) {constB=7}
  n14 = Constant() {r=0.5}
  n15 = SphereMask(A=TextureCoordinate() {coordinateIndex=1}, B=@n14, Radius=@n14, Hardness=LinearInterpolate(Alpha=@n5.Size) {constA=1, constB=0}) {hardnessPercent=25}
  n16 = Multiply(
   A=Multiply(A=LinearInterpolate(A=@n4, Alpha=@n5.Flicker) {constA=1}, B=@n6.RGB)
   B=Multiply(
    A=Multiply(A=@n12, B=@n13) {constB=80}
    B=TextureSample(UVs=@n15) {texture=/Game/Effects/Fort_Effects/Textures/Grad/T_Halo_Corona_Gradient}.RGB
   )
  )
  n17 = MaterialFunctionCall(LDR Input=Clamp(None=@n16), HDR Tint=@n16) {materialFunction=&DeriveHDRfromLDR}
  n18 = ScalarParameter() {parameterName=Desat}
  MaterialFunctionCall(Emissive=MaterialFunctionCall(Hue Shift Percentage=ScalarParameter() {parameterName=HueS}, Texture=Desaturation(None=@n17.RemappedRange, Fraction=@n18)) {materialFunction=&HueShift}.Result) {materialFunction=&MF_FX_EmissiveAdjustments}
  n19 = DepthFade(
   Opacity=MaterialFunctionCall(Distance=ScalarParameter() {defaultValue=750, parameterName=NearCamFade}) {materialFunction=&Near_Camera_Fade}.Result
   FadeDistance=ScalarParameter() {defaultValue=100, parameterName=DepthFade}
  )
  n20 = Constant() {r=1}
  n21 = ComponentMask(None=TextureCoordinate() {coordinateIndex=1}) {g=true}
  Multiply(
   A=Multiply(A=@n6.A, B=ShadingPathSwitch(Default=@n19, Mobile=@n20) {desc=Migrated from FeatureLevelSwitch})
   B=Power(Base=OneMinus(None=@n21), Exp=ScalarParameter() {defaultValue=10, parameterName=2341})
  )
  Subtract(A=@n9.RGB, B=@n11.RGB)
 ]
}
