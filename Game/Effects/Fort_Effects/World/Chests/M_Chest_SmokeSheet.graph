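# Unlit translucent mesh-particle material for scrolling smoke sheets in chest effects.
# Two differently scaled, panning samples of a tiling smoke texture are summed, sharpened by an exponent and shaped by a UV vertical gradient.
# Vertex red, particle alpha, a side mask that the "Safe?" switch can bypass, and Fresnel further mask this, feeding a clamped depth fade for soft intersections.
# Particle color supplies emissive through a shared adjustment function.
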Material /Game/Effects/Fort_Effects/World/Chests/M_Chest_SmokeSheet
graph {
 properties {bUsedWithMeshParticles=true, blendMode=BLEND_Translucent, pixelDepthOffsetMode=PDOM_Legacy, shadingModel=MSM_Unlit, usedShadingModels=Unlit}
 nodes [
  n1 = ParticleColor()
  n2 = VertexColor()
  n3 = ComponentMask(None=TextureCoordinate()) {g=true}
  n4 = ScalarParameter() {defaultValue=2, parameterName=Gradient Exponent}
  n5 = TextureCoordinate()
  n6 = Multiply(A=@n5, B=AppendVector(A=ScalarParameter() {defaultValue=0.4, parameterName=U1}, B=ScalarParameter() {defaultValue=0.4, parameterName=V1}))
  n7 = Multiply(A=Time(), B=ScalarParameter() {defaultValue=1, parameterName=Panner Speed})
  n8 = Multiply(A=@n5, B=AppendVector(A=ScalarParameter() {defaultValue=0.38, parameterName=U2}, B=ScalarParameter() {defaultValue=0.22, parameterName=V2}))
  n9 = Add(
   A=TextureSample(UVs=Panner(Coordinate=@n6, Time=@n7) {bFractionalPart=true, speedX=0.0121, speedY=0.0534}) {texture=/Game/Global/Textures/Default/Tileable_Patterns/T_Smoke_Puff_Tile_Up}.RGB
   B=TextureSample(UVs=Panner(Coordinate=@n8, Time=@n7) {bFractionalPart=true, speedX=-0.0155, speedY=0.05}) {texture=&T_Smoke_Puff_Tile_Up}.RGB
  )
  n10 = Multiply(
   A=Multiply(A=Power(Base=@n3, Exp=@n4), B=ScalarParameter() {defaultValue=10, parameterName=Gradient Multiplier}) {constB=10}
   B=Power(Base=Divide(A=@n9), Exp=ScalarParameter() {defaultValue=3, parameterName=Cloud Exponent}) {constExponent=3}
  )
  n11 = Constant() {r=1}
  n12 = TextureSample() {texture=/Game/Effects/Fort_Effects/Textures/Masks/T_Side_Mask}
  n13 = Multiply(A=@n1.A, B=Multiply(A=Multiply(A=@n2.R, B=@n10), B=StaticSwitchParameter(True=@n11, False=@n12.RGB) {parameterName=Safe?}))
  n14 = MaterialFunctionCall(
   Brightness=ScalarParameter() {defaultValue=1, parameterName=Fresnel Brightness}
   Exponent=ScalarParameter() {defaultValue=2, parameterName=Fresnel Exponent}
  ) {materialFunction=&MF_Fresnel_FX}
  DepthFade(Opacity=Clamp(None=Multiply(A=@n13, B=@n14.Fresnel)), FadeDistance=ScalarParameter() {defaultValue=20, parameterName=Depth Fade Distance})
  MaterialFunctionCall(Emissive=@n1.RGB) {materialFunction=&MF_FX_EmissiveAdjustments}
 ]
}
