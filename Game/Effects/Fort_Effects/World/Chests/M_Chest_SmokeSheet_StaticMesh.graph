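# Translucent unlit smoke-sheet material for chest effects, built for static meshes and mesh particles with warm orange glow colors.
# Two panned smoke textures, shaped by UV gradients and vertex red, blend two emissive colors with optional particle and time-of-day tinting.
# Opacity combines that mask with edge falloff, Fresnel, depth fade, particle alpha and a dynamic-parameter dissolve.
# Low quality uses separate mobile colors and skips depth fade.
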Material /Game/Effects/Fort_Effects/World/Chests/M_Chest_SmokeSheet_StaticMesh
graph {
 properties {
  bUsedWithMeshParticles=true
  bUsedWithNanite=true
  bUsedWithNiagaraMeshParticles=true
  blendMode=BLEND_Translucent
  pixelDepthOffsetMode=PDOM_Legacy
  shadingModel=MSM_Unlit
  usedShadingModels=Unlit
 }
 nodes [
  n1 = VectorParameter() {group=Time Of Day, parameterName=TimeOfDayMultipliers, defaultValue=<rgba 1 1 1 1>}
  n2 = MaterialFunctionCall(Value_Morning=@n1.R, Value_Day=@n1.R, Value_Evening=@n1.R, Value_Night=@n1.R) {materialFunction=&Scalar_Time_of_Day}
  n3 = ParticleColor()
  n4 = ComponentMask(None=TextureCoordinate()) {g=true}
  n5 = Power(Base=OneMinus(None=@n4), Exp=ScalarParameter() {defaultValue=2, parameterName=Gradient Two Exponent})
  n6 = ScalarParameter() {defaultValue=10, parameterName=Gradient Two Multiplier}
  n7 = Multiply(
   A=Power(Base=@n4, Exp=ScalarParameter() {defaultValue=2, parameterName=Gradient Exponent})
   B=ScalarParameter() {defaultValue=10, parameterName=Gradient Multiplier}
  ) {constB=10}
  n8 = StaticSwitchParameter(True=Multiply(A=Multiply(A=@n5, B=@n6) {constB=10}, B=@n7), False=@n7) {parameterName=UseSecondaryGradient}
  n9 = TextureCoordinate()
  n10 = ScalarParameter() {defaultValue=0.4, parameterName=U1}
  n11 = ScalarParameter() {defaultValue=0.4, parameterName=V1}
  n12 = Panner(Coordinate=Multiply(A=@n9, B=AppendVector(A=@n10, B=@n11))) {bFractionalPart=true, speedX=0.0121, speedY=0.0534}
  n13 = ScalarParameter() {defaultValue=0.38, parameterName=U2}
  n14 = ScalarParameter() {defaultValue=0.22, parameterName=V2}
  n15 = Panner(Coordinate=Multiply(A=@n9, B=AppendVector(A=@n13, B=@n14))) {bFractionalPart=true, speedX=-0.0155, speedY=0.05}
  n16 = Divide(A=Add(
    A=TextureSample(UVs=@n12) {texture=/Game/Global/Textures/Default/Tileable_Patterns/T_Smoke_Puff_Tile_Up}.RGB
    B=TextureSample(UVs=@n15) {texture=&T_Smoke_Puff_Tile_Up}.RGB
   ))
  n17 = ScalarParameter() {defaultValue=2.5, parameterName=Cloud Exponent}
  n18 = Multiply(A=VertexColor().R, B=Multiply(A=Saturate(None=@n8), B=Power(Base=@n16, Exp=@n17) {constExponent=3}))
  n19 = QualitySwitch(
   Default=LinearInterpolate(
    A=VectorParameter() {parameterName=Color A, defaultValue=<rgba 3 0.8856 0.1719 1>}.RGB
    B=VectorParameter() {parameterName=Color B, defaultValue=<rgba 8 2.4549 0.5833 1>}.RGB
    Alpha=@n18
   )
   Low=LinearInterpolate(
    A=VectorParameter() {parameterName=Color A Mobile, defaultValue=<rgba 1 0.2952 0.0573 1>}.RGB
    B=VectorParameter() {parameterName=Color B Mobile, defaultValue=<rgba 1 0.3069 0.0729 1>}.RGB
    Alpha=@n18
   )
  )
  n20 = StaticSwitchParameter(True=Multiply(A=@n3.RGB, B=@n19), False=@n19) {parameterName=UseForParticles?}
  MaterialFunctionCall(Emissive=StaticSwitchParameter(True=Multiply(A=@n2.Result, B=@n20), False=@n20) {parameterName=UseTimeOfDayMultiplier}) {materialFunction=&MF_FX_EmissiveAdjustments}
  n21 = ComponentMask(None=TextureCoordinate()) {r=true}
  n22 = Constant() {r=0.5}
  n23 = Constant() {r=0.45}
  n24 = Constant() {r=0.7}
  n25 = TextureCoordinate()
  n26 = Multiply(
   A=Power(Base=ComponentMask(None=@n25) {g=true}, Exp=ScalarParameter() {defaultValue=5.2608, parameterName=Outer Edge Contrast}) {constExponent=5}
   B=ScalarParameter() {defaultValue=50, parameterName=Outer Edge Scalar}
  ) {constB=10}
  n27 = StaticSwitchParameter(True=Constant() {r=1}, False=Multiply(A=SphereMask(A=@n21, B=@n22, Radius=@n23, Hardness=@n24), B=Saturate(None=@n26))) {parameterName=Safe?}
  n28 = ScalarParameter() {defaultValue=1, parameterName=Fresnel Brightness}
  n29 = ScalarParameter() {defaultValue=2, parameterName=Fresnel Exponent}
  n30 = Clamp(None=Multiply(
    A=Multiply(A=@n18, B=@n27)
    B=MaterialFunctionCall(Brightness=@n28, Exponent=@n29) {materialFunction=&MF_Fresnel_FX}.Fresnel
   ))
  n31 = DepthFade(Opacity=@n30, FadeDistance=ScalarParameter() {defaultValue=50, parameterName=Depth Fade Distance})
  n32 = Multiply(A=@n3.A, B=QualitySwitch(Default=@n31, Low=Multiply(A=@n30) {constB=3}, High=@n31, Medium=@n31))
  n33 = TextureSample() {texture=&T_Smoke_Puff_Tile_Up}
  n34 = DynamicParameter() {paramNames [DissolveExponent, Param2, Param3, Param4]}
  StaticSwitchParameter(True=Multiply(A=@n32, B=Power(Base=@n33.RGB, Exp=@n34.DissolveExponent) {constExponent=10}), False=@n32) {defaultValue=true, parameterName=UseDissolve?}
 ]
}
