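# Translucent, unlit, two-sided smoke sheet material for chest effects on static meshes or Niagara mesh particles.
# Two panned smoke textures, shaped by a V gradient and vertex red, form a mask blending two emissive colors, with separate low-quality colors.
# That mask feeds opacity with Fresnel, a switchable U-edge mask and depth fade, replaced on mobile by time-of-day scaling.
# Panned distortion textures, vertex normals and vertex red build a displacement offset.
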Material /Game/Effects/Fort_Effects/World/Chests/M_Chest_SmokeSheet_StaticMesh_V2
graph {
 properties {
  bUsedWithInstancedStaticMeshes=true
  bUsedWithNanite=true
  bUsedWithNiagaraMeshParticles=true
  blendMode=BLEND_Translucent
  pixelDepthOffsetMode=PDOM_Legacy
  shadingModel=MSM_Unlit
  twoSided=true
  usedShadingModels=Unlit
 }
 nodes [
  n1 = VectorParameter() {parameterName=Color A, defaultValue=<rgba 1 0.7833 0.5 1>}
  n2 = VectorParameter() {parameterName=Color B, defaultValue=<rgba 8 2.3179 0.4 1>}
  n3 = VertexColor()
  n4 = TextureCoordinate()
  n5 = Multiply(
   A=Power(Base=ComponentMask(None=@n4) {g=true}, Exp=ScalarParameter() {defaultValue=6, parameterName=Gradient Exponent})
   B=ScalarParameter() {defaultValue=10, parameterName=Gradient Multiplier}
  ) {constB=10}
  n6 = TextureCoordinate()
  n7 = ScalarParameter() {defaultValue=0.75, parameterName=U1}
  n8 = ScalarParameter() {defaultValue=0.5, parameterName=V1}
  n9 = Multiply(A=Time(), B=ScalarParameter() {defaultValue=2.5, parameterName=Time Scale Cloud Detail})
  n10 = Panner(Coordinate=Multiply(A=@n6, B=AppendVector(A=@n7, B=@n8)), Time=@n9) {bFractionalPart=true, speedX=0.0121, speedY=0.0534}
  n11 = ScalarParameter() {defaultValue=1, parameterName=U2}
  n12 = ScalarParameter() {defaultValue=1, parameterName=V2}
  n13 = Panner(Coordinate=Multiply(A=@n6, B=AppendVector(A=@n11, B=@n12)), Time=@n9) {bFractionalPart=true, speedX=-0.0155, speedY=0.05}
  n14 = Divide(A=Add(
    A=TextureSample(UVs=@n10) {texture=/Game/Global/Textures/Default/Tileable_Patterns/T_Smoke_Puff_Tile_Up}.RGB
    B=TextureSample(UVs=@n13) {texture=&T_Smoke_Puff_Tile_Up}.RGB
   ))
  n15 = ScalarParameter() {defaultValue=1, parameterName=Cloud Exponent}
  n16 = Multiply(A=@n3.R, B=Multiply(A=Saturate(None=@n5), B=Power(Base=@n14, Exp=@n15) {constExponent=3}))
  n17 = VectorParameter() {parameterName=Color A Mobile, defaultValue=<rgba 1 0.6262 0.5 1>}
  n18 = VectorParameter() {parameterName=Color B Mobile, defaultValue=<rgba 2 0.5795 0.1 1>}
  MaterialFunctionCall(Emissive=QualitySwitch(Default=LinearInterpolate(A=@n1.RGB, B=@n2.RGB, Alpha=@n16), Low=LinearInterpolate(A=@n17.RGB, B=@n18.RGB, Alpha=@n16))) {materialFunction=&MF_FX_EmissiveAdjustments}
  n19 = VectorParameter() {group=WPO, parameterName=TextureWPO_UVtiling, defaultValue=<rgba 2 2 0 0>}
  n20 = Panner(
   Coordinate=Multiply(A=TextureCoordinate(), B=ComponentMask(None=@n19.RGB) {g=true, r=true})
   Time=Multiply(A=Time(), B=ScalarParameter() {defaultValue=0.5, parameterName=Time Scale})
  ) {speedY=1}
  n21 = TextureSample(UVs=@n20) {samplerType=SAMPLERTYPE_Normal, texture=&T_Weap_Distort_N_02}
  n22 = TextureSample(UVs=@n20) {texture=&T_Smooth_tile}
  n23 = VectorParameter() {parameterName=WPO Direction, defaultValue=<rgba 20 20 0 0>}
  Multiply(
   A=Multiply(A=LinearInterpolate(A=@n21.RGB, B=@n22.RGB), B=VertexNormalWS())
   B=LinearInterpolate(A=ComponentMask(None=@n23.RGB) {b=true, g=true, r=true}, Alpha=@n3.R)
  )
  n24 = TextureCoordinate()
  Multiply(A=Power(Base=ComponentMask(None=@n24) {g=true}, Exp=ScalarParameter() {defaultValue=-10, parameterName=Fade Out}))
  n25 = TextureCoordinate()
  n26 = TextureCoordinate()
  n27 = StaticSwitchParameter(
   True=Constant() {r=1}
   False=SphereMask(A=ComponentMask(None=@n26) {r=true}, B=Constant() {r=0.5}, Radius=Constant() {r=0.45}, Hardness=Constant() {r=0.7})
  ) {parameterName=Safe?}
  n28 = ScalarParameter() {defaultValue=1, parameterName=Fresnel Brightness}
  n29 = ScalarParameter() {defaultValue=2, parameterName=Fresnel Exponent}
  n30 = Clamp(None=Multiply(
    A=Multiply(A=@n16, B=@n27)
    B=MaterialFunctionCall(Brightness=@n28, Exponent=@n29) {materialFunction=&MF_Fresnel_FX}.Fresnel
   ))
  n31 = ScalarParameter() {defaultValue=50, parameterName=Depth Fade Distance}
  n32 = MaterialFunctionCall(Value_Morning=Constant() {r=1}, Value_Day=Constant() {r=0.75}, Value_Evening=Constant() {r=1}, Value_Night=Constant() {r=1}) {materialFunction=&Scalar_Time_of_Day}
  Multiply(
   A=Power(Base=ComponentMask(None=@n25) {g=true}, Exp=ScalarParameter() {parameterName=Exp})
   B=ShadingPathSwitch(Default=DepthFade(Opacity=@n30, FadeDistance=@n31), Mobile=Multiply(A=@n30, B=@n32.Result)) {desc=Migrated from FeatureLevelSwitch}
  )
  n33 = Constant() {r=0.5}
  SphereMask(
   A=Power(
    Base=MaterialFunctionCall() {materialFunction=&LinearGradient}.VGradient
    Exp=ScalarParameter() {defaultValue=0.5, group=WPO, parameterName=CenterOffset}
   ) {constExponent=0.5}
   B=@n33
   Radius=@n33
  ) {hardnessPercent=0}
 ]
}
